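# M_Trail is an empty material in the FX folder of the Glider_Racer_Zero_Male glider cosmetic, named for a trail effect.
# It is flagged for use with Niagara ribbons, uses the Default Lit shading model and sets pixel depth offset to legacy mode.
# Its graph holds no expressions, so it defines no color, opacity or animation logic of its own.
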Material /BRCosmetics/Gadgets/Assets/VinderTech_GliderChute/Glider_Racer_Zero_Male/FX/M_Trail
graph {expressions {}, properties {bUsedWithNiagaraRibbons=true, pixelDepthOffsetMode=PDOM_Legacy, usedShadingModels=Default Lit}}
